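# Ambient spot sound cue playing magic whisper one-shots for the Seven Meeting Site point of interest.
# A quality-level root routes its first two slots to a random pick from eight whispers and its last two to a smaller set of three.
# Each random set feeds a modulator that varies pitch and volume.
# The cue uses the Athena outdoor ambient sound class and borrows a bird one-shot attenuation preset.
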SoundCue /Game/Sounds/Artemis/POI/Seven_Meeting_Site/Spots/Spot_Seven_Meeting_Site_Magic_Whisper_Cue
root=QualityLevel_0
nodes [
 {id=Modulator_1, children [Random_3], properties {pitchMax=1.1, volumeMax=1.1, volumeMin=0.9}}
 {id=Modulator_2, children [Random_2], properties {pitchMax=1.1, volumeMax=1.1, volumeMin=0.9}}
 {id=QualityLevel_0, children [Modulator_1,Modulator_1,Modulator_2,Modulator_2]}
 {id=Random_2, children [WavePlayer_0,WavePlayer_1,WavePlayer_2], properties {weights [1,1,1]}}
 {
  id=Random_3
  children [WavePlayer_3, WavePlayer_4, WavePlayer_5, WavePlayer_14, WavePlayer_15, WavePlayer_16, WavePlayer_17, WavePlayer_18]
  properties {weights [1,1,1,1,1,1,1,1]}
 }
 {id=WavePlayer_0, properties {soundWaveAssetPtr=&/Spot_Seven_Meeting_Site_Magic_Whisper_02}}
 {id=WavePlayer_1, properties {soundWaveAssetPtr=&/Spot_Seven_Meeting_Site_Magic_Whisper_05}}
 {id=WavePlayer_14, properties {soundWaveAssetPtr=&/Spot_Seven_Meeting_Site_Magic_Whisper_03}}
 {id=WavePlayer_15, properties {soundWaveAssetPtr=&/Spot_Seven_Meeting_Site_Magic_Whisper_04}}
 {id=WavePlayer_16, properties {soundWaveAssetPtr=&/Spot_Seven_Meeting_Site_Magic_Whisper_06}}
 {id=WavePlayer_17, properties {soundWaveAssetPtr=&/Spot_Seven_Meeting_Site_Magic_Whisper_07}}
 {id=WavePlayer_18, properties {soundWaveAssetPtr=&/Spot_Seven_Meeting_Site_Magic_Whisper_08}}
 {id=WavePlayer_2, properties {soundWaveAssetPtr=&Spot_Seven_Meeting_Site_Magic_Whisper_07}}
 {id=WavePlayer_3, properties {soundWaveAssetPtr=&Spot_Seven_Meeting_Site_Magic_Whisper_05}}
 {id=WavePlayer_4, properties {soundWaveAssetPtr=&/Spot_Seven_Meeting_Site_Magic_Whisper_01}}
 {id=WavePlayer_5, properties {soundWaveAssetPtr=&Spot_Seven_Meeting_Site_Magic_Whisper_02}}
]
properties {
 attenuationSettings=&Fort_TileAmbient_OneShot_Bird_Att
 maxDistance=8005
 soundClassObject=&Fort_Ambient_Outdoors_Athena
 volumeMultiplier=0.752
 preEffectBusSends [{}]
}
